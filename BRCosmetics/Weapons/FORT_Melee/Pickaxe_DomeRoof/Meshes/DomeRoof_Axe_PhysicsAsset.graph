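# Physics asset for the DomeRoof pickaxe, a two-handed melee cosmetic weapon with a dangling chain piece.
# A capsule body on the handle bone and a smaller capsule on the first dynamic chain bone form the collision shapes.
# One constraint joins the chain body to the handle with limited swing and twist.
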PhysicsAsset /BRCosmetics/Weapons/FORT_Melee/Pickaxe_DomeRoof/Meshes/DomeRoof_Axe_PhysicsAsset
bodies {
 Melee_TwoHanded_Handle [{shapeType=Capsule, length=48.9106, radius=9.9763, center=<2.6746 0.2772 34.1408>, rotation=<rot 5.6369 -90 0.3042>}]
 dyn_chain_01_A [{shapeType=Capsule, length=12.1283, radius=5.4242, center=<11.8175 -1.2229 0.0746>, rotation=<rot 0 -104.8457 -89.6759>}]
}
constraints [{bone1Name=dyn_chain_01_A, bone2Name=Melee_TwoHanded_Handle, swing1LimitDegrees=45, swing2LimitDegrees=45, twistLimitDegrees=45}]
